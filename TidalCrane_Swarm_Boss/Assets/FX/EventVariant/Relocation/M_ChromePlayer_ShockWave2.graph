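# Translucent, unlit, two-sided material for Niagara mesh particles, drawing a smoky shockwave effect.
# A smoke trail texture samples on swapped UV axes, distorted by panning Perlin noise.
# Its red channel, particle alpha and depth fade form an opacity mask; the same channel, remapped, scales adjusted particle color and a time-of-day tint.
# SplineThicken widens the mesh strip through a Width parameter.
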Material /TidalCrane_Swarm_Boss/Assets/FX/EventVariant/Relocation/M_ChromePlayer_ShockWave2
graph {
 properties {
  bUsedWithNiagaraMeshParticles=true
  blendMode=BLEND_Translucent
  pixelDepthOffsetMode=PDOM_Legacy
  shadingModel=MSM_Unlit
  twoSided=true
  usedShadingModels=Unlit
 }
 nodes [
  n1 = ScalarParameter() {defaultValue=1000, parameterName=Width}
  MaterialFunctionCall(WidthBase=@n1, WidthTip=@n1) {materialFunction=&SplineThicken}
  n2 = TextureObjectParameter() {
   group=MainTexture
   parameterName=MainTexture_Texture
   sortPriority=0
   texture=/Game/Effects/Fort_Effects/Textures/Smoke/T_Puffy_Smoke_Trail_01
  }
  n3 = TextureObjectParameter() {
   group=NoiseA
   parameterName=NoiseA_Texture
   samplerType=SAMPLERTYPE_LinearGrayscale
   sortPriority=0
   texture=&T_SimplePerlin_Greyscale
  }
  n4 = VectorParameter() {group=NoiseA, parameterName=NoiseA_UVControl, channelNames {a=VPan, b=UPan, g=VScale, r=UScale}, defaultValue=<rgba 1 2 -1 0>}
  n5 = VectorParameter() {
   group=NoiseA
   parameterName=NoiseA_NoiseControl
   sortPriority=60
   channelNames {a=VStrength, b=UStrength, g=VOffset, r=UOffset}
   defaultValue=<rgba -0.5 -0.5 0 0.16>
  }
  n6 = VectorParameter() {group=MainTexture, parameterName=MainTexture_UVControl, channelNames {a=VPan, b=UPan, g=VScale, r=UScale}, defaultValue=<rgba 1 1 0 0>}
  n7 = MaterialFunctionCall(
   Texture=@n2
   AdditionalUVOffset=MaterialFunctionCall!(
    NoiseValue=MaterialFunctionCall(Texture=@n3, UVControl_RGB=@n4, UVControl_A=@n4) {materialFunction=&FXCommon_TextureSample}.RG
    Control_RGB=@n5
    Control_A=@n5
   ) {materialFunction=/Game/Effects/Niagara/MasterMaterials/Functions/Common/FXCommon_NoiseControl}.NoiseValue
   UVControl_RGB=@n6
   UVControl_A=@n6
   "[Override]UVInput"=MaterialFunctionCall(XY=TextureCoordinate()) {materialFunction=&Swizzle}.YX
  ) {materialFunction=&FXCommon_TextureSample}
  n8 = ParticleColor()
  Multiply(A=Multiply(A=@n7.R, B=@n8.A), B=DepthFade(FadeDistance=ScalarParameter() {defaultValue=1400, parameterName=DF}))
  n9 = VectorParameter() {group=Sunlight- Time of Day, parameterName="Color_Morning (V3)", defaultValue=<rgba 0.1739 0.1137 0.0967 1>}
  n10 = VectorParameter() {group=Sunlight- Time of Day, parameterName="Color_Evening (V3)", defaultValue=<rgba 0.2922 0.1742 0.1378 1>}
  n11 = Constant()
  n12 = Constant() {r=1}
  n13 = ScalarParameter() {defaultValue=0.2, parameterName=LOW}
  Multiply(
   A=MaterialFunctionCall(
    Color_Morning=@n9.RGB
    Color_Day=VectorParameter() {group=Sunlight- Time of Day, parameterName="Color_Day (V3)", defaultValue=<rgba 0.2783 0.2239 0.1907 1>}.RGB
    Color_Evening=@n10.RGB
    Color_Night=VectorParameter() {group=Sunlight- Time of Day, parameterName="Color_Night (V3)", defaultValue=<rgba 0.1214 0.1428 0.25 1>}.RGB
   ) {materialFunction=&Color_Time_of_Day}.Result
   B=Multiply(
    A=MaterialFunctionCall(Input=@n7.R, Input Low=@n11, Input High=@n12, Target Low=@n13, Target High=@n12) {materialFunction=&RemapValueRange}.Result
    B=MaterialFunctionCall(Emissive=@n8.RGB) {materialFunction=&MF_FX_EmissiveAdjustments}.Result
   )
  )
 ]
}
