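# Niagara effect for the TidalCrane swarm boss event's reposition, showing a large ground eruption of cracks, rocks, smoke and dirt.
# A short smoke stream leads into bursts at half a second: a growing, fading crack decal, tumbling mesh rocks, smoke puffs and dirt sprites.
# Spawn groups give rocks and smoke two variants each, leaving their duplicate emitters disabled.
# Time-of-day colors tint smoke and dirt; IsUltra and FX_Scalar affect only dirt.
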NiagaraSystem /TidalCrane_Swarm_Boss/Assets/FX/EventVariant/Relocation/NS_TidalCrane_Event_RepositionInto
dependencies {
 usedDataInterfaces [
  &NiagaraDataInterfaceCurve
  &NiagaraDataInterfaceSpriteRendererInfo
  &NiagaraDataInterfaceColorCurve
  &NiagaraDataInterfaceMeshRendererInfo
  &NiagaraDataInterfaceVector2DCurve
 ]
 usedDynamicInputs [&Vector2DFromFloat, &SelectVectorFromArray]
 usedModules [&SystemState]
}
emitters {
 Ground_Cracks {
  emitterUpdateScript [
   &EmitterState()
   &SpawnBurst_Instantaneous(Spawn Time=0.5, Use Spawn Probability=false, UseLoopCountLimit=false)
  ]
  particleSpawnScript [
   &InitializeParticle(Lifetime=3, Sprite Rotation Mode=::NewEnumerator2, UsePositionOffset=true, Write Material Random=true) {module=InitializeParticle001}
   /SetVariables(Particles.RandomDecalYZ=5261.5962) {isSetParameters=true}
   /SetVariables(Particles.DecalSize:Vector3f=&MakeVector(X=91.2157, Y=$Particles.RandomDecalYZ, Z=$Particles.RandomDecalYZ)) {isSetParameters=true}
   &SpriteFacingAndAlignment(Sprite Alignment=<1 0 0>, Sprite Facing=<0 0 1>, WriteAlignment=true, WriteFacing=true) {module=SpriteFacingAndAlignment001}
  ]
  particleUpdateScript [
   &ParticleState()
   /SetVariables(Particles.DecalFade:NiagaraFloat=&FloatFromCurve(FloatCurve=DataInterface() {propertyValues {Curve {keys [
         {interpMode=RCIM_Cubic, tangentMode=RCTM_User, time=0, value=0}
         <at 0.1786 1>
         {arriveTangent=-1.2175, interpMode=RCIM_Cubic, leaveTangent=-1.2175, time=0.7, value=1}
         <at 1 0 RCIM_Cubic>
        ]}}})) {isSetParameters=true}
   /SetVariables(Particles.DecalSize:Vector3f=&Multiply_VectorByFloat(
     Vector=$Particles.Initial.DecalSize
     Float=&FloatFromCurve(FloatCurve=DataInterface() {propertyValues {Curve {keys [{arriveTangent=139.099, interpMode=RCIM_Cubic, leaveTangent=139.099, tangentMode=RCTM_User, time=0, value=0}, <at 0.015 1 RCIM_Cubic>]}}})
    )) {isSetParameters=true}
  ]
  rendererClasses [&NiagaraDecalRendererProperties]
  renderers [{rendererClass=&NiagaraDecalRendererProperties, rendererIndex=0}]
 }
 Sharp_Rocks {
  emitterUpdateScript [
   &EmitterState()
   &SpawnBurst_Instantaneous(Spawn Count=100, Spawn Time=0.5, Use Spawn Probability=false, UseLoopCountLimit=false)
   &SpawnBurst_Instantaneous(Spawn Count=15, Spawn Group=1, Spawn Time=0.5, Use Spawn Probability=false, UseLoopCountLimit=false) {module=SpawnBurst_Instantaneous001}
  ]
  particleSpawnScript [
   &InitializeParticle(
    Color=<rgba 1 0.8667 0.8 1>
    Lifetime=&MergeHelper_FloatSwitch(
     0=&RandomRangeFloat(Minimum=1, Maximum=2)
     1=&RandomRangeFloat(Minimum=2, Maximum=3)
     2=0
     3=0
     4=0
     Index=&SpawnGroup()
     Num=::NewEnumerator3
    )
    Mesh Scale=&MergeHelper_VectorSwitch(
     0=&Multiply_VectorByFloat(
      Vector=&UniformRangedVector(Minimum=<0.15 0.3 0.4>, Maximum=<0.6 1 1.1>)
      Float=&UniformRangedFloat(Minimum=10, Maximum=15)
     )
     1=&Multiply_VectorByFloat(
      Vector=&UniformRangedVector(Minimum=<0.15 0.3 0.4>, Maximum=<0.2 1 1.1>)
      Float=&UniformRangedFloat(Minimum=54.5942, Maximum=64.3507)
     )
     2=<0 0 0>
     3=<0 0 0>
     4=<0 0 0>
     Index=&SpawnGroup()
     Num=::NewEnumerator3
    )
    Mesh Scale Mode=::NewEnumerator4
    Sprite Rotation Mode=::NewEnumerator2
    UsePositionOffset=false
    Write Material Random=true
   )
   &InitialMeshOrientation(Orientation Vector=&RandomVector(Vector Scale=1), Use Orientation Vector=true, Use Rotation Vector=false)
   /SetVariables(Particles.MeshRandom:NiagaraInt32=&RandomRangeInt(Minimum=0, Maximum=2)) {isSetParameters=true, module=SetVariables_#1}
   &ShapeLocation(
    Handle Radius:NiagaraFloat=&MergeHelper_FloatSwitch(0=108, 1=180, 2=0, 3=0, 4=0, Index=&SpawnGroup(), Num=::NewEnumerator3)
    Large Radius=&MergeHelper_FloatSwitch(0=370.4802, 1=384.5074, 2=0, 3=0, 4=0, Index=&SpawnGroup(), Num=::NewEnumerator3)
    MaskSpawnGroup=false
    Override Seed=false
    Rotation Angle=0.3
    Rotation Axis=<0 0 1>
    Rotation Mode=::NewEnumerator1
    Shape Primitive=::NewEnumerator3
    UseOwnerScale=true
   )
   &AddVelocityFromPoint(
    Module.Use Falloff Distance=false
    Origin Offset=<0 0 -500>
    Velocity Strength=&MergeHelper_FloatSwitch(
     0=&RandomRangeFloat(Minimum=3017.488, Maximum=5449.7114)
     1=&RandomRangeFloat(Minimum=1784.5364, Maximum=3000.7002)
     2=0
     3=0
     4=0
     Index=&SpawnGroup()
     Num=::NewEnumerator3
    )
   )
   &AddVelocity(Velocity=&MergeHelper_VectorSwitch(
     0=&RandomRangeVector(Minimum=<0 0 39.6334>, Maximum=<0 0 80.4828>)
     1=&RandomRangeVector(Minimum=<0 0 1149.2963>, Maximum=<0 0 1857.4417>)
     2=<0 0 0>
     3=<0 0 0>
     4=<0 0 0>
     Index=&SpawnGroup()
     Num=::NewEnumerator3
    ))
   &AddRotationalVelocity(Rotation Rate=&RandomVector(Vector Scale=&RandomRangeFloat(
      Minimum=1
      Maximum=&MergeHelper_FloatSwitch(0=3, 1=1.8, 2=0, 3=0, 4=0, Index=&SpawnGroup(), Num=::NewEnumerator3)
     )))
   &ApplyInitialForces()
   &DynamicMaterialParameters(Module.Index0Param0Write=true, Module.Index0Param1Write=false, Module.Index0Param2Write=false, Module.Index0Param3Write=false)
   &CameraOffset(Camera Offset Amount=&MergeHelper_FloatSwitch(0=347.3134, 1=900, 2=0, 3=0, 4=0, Index=&SpawnGroup(), Num=::NewEnumerator3))
   /SetVariables(Particles.Velocity=&Multiply_Vector(
     A=$Particles.Velocity
     B=&MergeHelper_VectorSwitch(0=<1 1 0.7>, 1=<1 1 1>, 2=<0 0 0>, 3=<0 0 0>, 4=<0 0 0>, Index=&SpawnGroup(), Num=::NewEnumerator3)
    )) {isSetParameters=true}
  ]
  particleUpdateScript [
   &ParticleState()
   &GravityForce(Gravity=&MergeHelper_VectorSwitch(
     0=&RandomRangeVector(Minimum=<0 0 -1945.1377>, Maximum=<0 0 -3716.2026>)
     1=&RandomRangeVector(Minimum=<0 0 -3162.3899>, Maximum=<0 0 -5032.9409>)
     2=<0 0 0>
     3=<0 0 0>
     4=<0 0 0>
     Index=&SpawnGroup()
     Num=::NewEnumerator3
    ))
   &Drag(
    Drag=&MergeHelper_FloatSwitch(
     0=&RandomRangeFloat(Minimum=0.25, Maximum=1.2033)
     1=&RandomRangeFloat(Minimum=0.15, Maximum=0.3)
     2=0
     3=0
     4=0
     Index=&SpawnGroup()
     Num=::NewEnumerator3
    )
    Use Linear Drag=true
    Use Rotational Drag=true
   )
   &SolveForcesAndVelocity(Clamp Velocity=false, Limit Acceleration=false)
   &ScaleMeshSize(Scale Factor=&VectorFromFloat(Value=&FloatFromCurve(FloatCurve=DataInterface() {propertyValues {Curve {keys [<at 0.75 1 RCIM_Cubic>, {arriveTangent=-6.2711, interpMode=RCIM_Cubic, leaveTangent=-6.2711, tangentMode=RCTM_User, time=1, value=0}]}}})))
  ]
  rendererClasses [&NiagaraMeshRendererProperties]
  renderers [{rendererClass=&NiagaraMeshRendererProperties, rendererIndex=0}]
 }
 Sharp_Rocks001 {
  bEnabled=false
  emitterUpdateScript [
   &EmitterState(Scalability Mode=::NewEnumerator1)
   &SpawnBurst_Instantaneous(Spawn Count=15, Spawn Time=0.5, Use Spawn Probability=false, UseLoopCountLimit=false)
  ]
  particleSpawnScript [
   &InitializeParticle(
    Color=<rgba 1 0.8667 0.8 1>
    Lifetime Max=3
    Lifetime Min=2
    Lifetime Mode=::NewEnumerator1
    Mesh Scale=&Multiply_VectorByFloat(
     Vector=&UniformRangedVector(Minimum=<0.15 0.3 0.4>, Maximum=<0.2 1 1.1>)
     Float=&UniformRangedFloat(Minimum=54.5942, Maximum=64.3507)
    )
    Mesh Scale Mode=::NewEnumerator4
    Sprite Rotation Mode=::NewEnumerator2
    UsePositionOffset=false
    Write Material Random=true
   )
   &InitialMeshOrientation(Orientation Vector=&RandomVector(Vector Scale=1), Use Orientation Vector=true, Use Rotation Vector=false)
   /SetVariables(Particles.MeshRandom:NiagaraInt32=&RandomRangeInt(Minimum=0, Maximum=2)) {isSetParameters=true, module=SetVariables_#1}
   &ShapeLocation(
    Handle Radius=183.4485
    Large Radius=384.5074
    MaskSpawnGroup=false
    Override Seed=false
    Rotation Angle=0.3
    Rotation Axis=<0 0 1>
    Rotation Mode=::NewEnumerator1
    Shape Primitive=::NewEnumerator3
    UseOwnerScale=true
   )
   &AddVelocityFromPoint(
    Module.Use Falloff Distance=false
    Origin Offset=<0 0 -500>
    Velocity Strength=&RandomRangeFloat(Minimum=2222.7683, Maximum=3883.0671)
   )
   &AddVelocity(Velocity=&RandomRangeVector(Minimum=<0 0 1149.2963>, Maximum=<0 0 1857.4417>))
   &AddRotationalVelocity(Rotation Rate=&RandomVector(Vector Scale=&RandomRangeFloat(Minimum=1, Maximum=1.4793)))
   &ApplyInitialForces()
   &DynamicMaterialParameters(Module.Index0Param0Write=true, Module.Index0Param1Write=false, Module.Index0Param2Write=false, Module.Index0Param3Write=false)
   &CameraOffset(Camera Offset Amount=901.9825)
  ]
  particleUpdateScript [
   &ParticleState()
   &GravityForce(Gravity=&RandomRangeVector(Minimum=<0 0 -4958.2827>, Maximum=<0 0 -7119.5688>))
   &Drag(Drag=&RandomRangeFloat(Minimum=0.15, Maximum=0.3), Use Linear Drag=true, Use Rotational Drag=true)
   &SolveForcesAndVelocity(Clamp Velocity=false, Limit Acceleration=false)
   &ScaleMeshSize(Scale Factor=&VectorFromFloat(Value=&FloatFromCurve(FloatCurve=DataInterface() {propertyValues {Curve {keys [<at 0.75 1 RCIM_Cubic>, {arriveTangent=-6.2711, interpMode=RCIM_Cubic, leaveTangent=-6.2711, tangentMode=RCTM_User, time=1, value=0}]}}})))
  ]
  rendererClasses [&NiagaraMeshRendererProperties]
  renderers [{rendererClass=&NiagaraMeshRendererProperties, rendererIndex=0}]
 }
 Smoke_SphereCenter {
  emitterUpdateScript [
   &EmitterState(Life Cycle Mode=::NewEnumerator1, Loop Behavior=::NewEnumerator1, Loop Delay=0.05, Loop Duration=0.51, UseLoopDelay=false)
   &SpawnRate(SpawnRate=50, Use Spawn Probability=false)
   &SpawnBurst_Instantaneous(Spawn Count=30, Use Spawn Probability=false, UseLoopCountLimit=false) {enabled=false}
   &SpawnBurst_Instantaneous(Spawn Count=30, Spawn Time=0.25, Use Spawn Probability=false, UseLoopCountLimit=false) {enabled=false, module=SpawnBurst_Instantaneous001}
   &SpawnBurst_Instantaneous(Spawn Count=10, Spawn Group=1, Spawn Time=0.5, Use Spawn Probability=false, UseLoopCountLimit=false) {module=SpawnBurst_Instantaneous002}
  ]
  particleSpawnScript [
   &InitializeParticle(
    Color=&RandomRangeLinearColor(Minimum=<rgba 0.1094 0.0735 0.0567 1>, Maximum=<rgba 0.2031 0.1365 0.1053 1>)
    Lifetime=&MaskFloatBySpawnGroup(
     Masked Float=&RandomRangeFloat(Minimum=2, Maximum=3)
     Passthrough Float=&RandomRangeFloat(Minimum=1.5, Maximum=2)
    )
    Sprite Rotation Angle=&RandomRangeFloat(Minimum=-25, Maximum=25)
    Sprite Rotation Mode=::NewEnumerator0
    Sprite Size Mode=::NewEnumerator3
    Uniform Sprite Size=&MaskFloatBySpawnGroup(
     Masked Float=&RandomRangeFloat(Minimum=1600, Maximum=1800)
     Passthrough Float=&RandomRangeFloat(Minimum=1500, Maximum=1600)
    )
    UsePositionOffset=false
    Write Material Random=true
   )
   &ShapeLocation(
    Handle Radius=208.3537
    Large Radius=205.2423
    MaskSpawnGroup=false
    Non Uniform Scale=<1 1 0.5>
    Offset=<0 0 -250>
    Offset Mode=::NewEnumerator0
    Override Seed=true
    Random Seed=20
    Randomness Mode:ENiagaraRandomnessMode=::NewEnumerator1
    Shape Primitive=::NewEnumerator3
    UseOwnerScale=true
   )
   /SetVariables(Particles.1Position:NiagaraPosition=$Particles.Position) {isSetParameters=true}
   /SetVariables() {isSetParameters=true}
   &ShapeLocation(
    MaskSpawnGroup=false
    Offset Mode=::NewEnumerator0
    Override Seed=true
    Random Seed=153
    Randomness Mode:ENiagaraRandomnessMode=::NewEnumerator1
    Sphere Radius=1192.1179
    UseOwnerScale=true
   ) {module=ShapeLocation001}
   /SetVariables(Particles.Position=&MaskVectorBySpawnGroup(
     Masked Vector=Linked() {linkedVariable {name=Particles.1Position, type=NiagaraPosition}}
     Passthrough Vector=Linked() {linkedVariable {name=Particles.Position, type=NiagaraPosition}}
    )) {isSetParameters=true}
   &AddVelocity(
    Module.Use Falloff Distance=false
    Origin Offset=&MaskVectorBySpawnGroup(Masked Vector=<0 0 -602.6362>, Passthrough Vector=<0 0 -451.717>)
    Override Seed=false
    Velocity Mode=::NewEnumerator1
    Velocity Speed=&MaskFloatBySpawnGroup(
     Masked Float=&RandomRangeFloat(Minimum=3354.2996, Maximum=7144.5815)
     Passthrough Float=&RandomRangeFloat(Minimum=2811.1414, Maximum=4059.1914)
    )
   )
   /SetVariables(Particles.Velocity=&Multiply_Vector(A=$Particles.Velocity, B=&MaskVectorBySpawnGroup(Masked Vector=<1 1 0.5>, Passthrough Vector=<0.5 0.5 1.3344>))) {isSetParameters=true, module=SetVariables_#2}
  ]
  particleUpdateScript [
   &ParticleState()
   &GravityForce(Gravity=&MaskVectorBySpawnGroup(
     Masked Vector=&RandomRangeVector(Minimum=<0 0 -790.381>, Maximum=<0 0 -1053.36>)
     Passthrough Vector=&RandomRangeVector(Minimum=<0 0 -137.6014>, Maximum=<0 0 -176.7234>)
    ))
   &Drag(
    Drag=&MaskFloatBySpawnGroup(
     Masked Float=&RandomRangeFloat(Minimum=1.2, Maximum=1.3)
     Passthrough Float=&RandomRangeFloat(Minimum=1.5, Maximum=2)
    )
    Use Linear Drag=true
    Use Rotational Drag=true
   )
   &SolveForcesAndVelocity(Clamp Velocity=false, Limit Acceleration=false)
   &ScaleSpriteSize(Uniform Curve Sprite Scale=DataInterface() {propertyValues {Curve {keys [
        {arriveTangent=3.9789, interpMode=RCIM_Cubic, leaveTangent=3.9789, tangentMode=RCTM_User, time=0, value=0.5}
        {arriveTangent=1.5, interpMode=RCIM_Cubic, leaveTangent=1.5, time=0.1337, value=1.7145}
        {arriveTangent=0.0224, interpMode=RCIM_Cubic, leaveTangent=0.0224, tangentMode=RCTM_User, time=1, value=2}
       ]}}})
   &DynamicMaterialParameters(
    Index 0 Param 1=&Lerp_Float(
     A=1
     B=2
     Alpha=&FloatFromCurve(FloatCurve=DataInterface() {propertyValues {Curve {keys [<at 0 0 RCIM_Cubic>,<at 1 1 RCIM_Cubic>]}}})
    )
    Module.Index0Param0Write=true
    Module.Index0Param1Write=true
    Module.Index0Param2Write=true
    Module.Index0Param3Write=true
   )
   &SubUVAnimation(
    Frames Per Second:NiagaraInt32=&MaskIntBySpawnGroup(Masked Int=&RandomRangeInt(Minimum=10, Maximum=15), Passthrough Int=20, Spawn Group Mask=0)
    Playback Mode:ENiagara_FPSPlayrate=::NewEnumerator1
    Playback Scale Curve=DataInterface() {propertyValues {Curve {keys [
        {interpMode=RCIM_Cubic, tangentMode=RCTM_User, time=0, value=3}
        {arriveTangent=-1.2656, interpMode=RCIM_Cubic, leaveTangent=-1.2656, tangentMode=RCTM_User, time=1, value=0}
       ]}}}
    Random Start Frame=true
    "Sprite Renderer "=DataInterface() {propertyValues {SpriteRenderer=&self:Smoke_SphereCenter_0.NiagaraSpriteRendererProperties_1}}
    Start Frame Offset:NiagaraInt32=&RandomRangeInt(Minimum=0, Maximum=6)
    Start Frame Range Override=&RandomRangeInt(Minimum=0, Maximum=0)
    SubUV Animation Mode:ENiagara_SubUVLookupModeV2=::NewEnumerator3
    UseEndFrame=true
    UsePlayRateScale=true
    UseStartFrame=true
   )
   &ScaleColor(
    Scale RGB=&MakeVectorFromLinearColorRGB(Color=&FNTimeOfDayColor(
      Night Color=<rgba 0.6 0.7243 1 1>
      Morning Color=<rgba 0.7 0.6725 0.6692 1>
      Day Color=<rgba 1 0.8103 0.65 1>
      Evening Color=<rgba 0.8 0.7149 0.6539 1>
     ))
    ScaleA=true
    ScaleRGB=true
   ) {module=ScaleColor001}
   &ScaleColor(
    Scale Alpha=&FloatFromCurve(FloatCurve=DataInterface() {propertyValues {Curve {keys [
         {arriveTangent=0.7573, interpMode=RCIM_Cubic, leaveTangent=0.7573, tangentMode=RCTM_User, time=0, value=0}
         {arriveTangent=-0.3388, interpMode=RCIM_Cubic, leaveTangent=-0.3388, tangentMode=RCTM_User, time=0.02, value=1}
         {arriveTangent=-2.7546, interpMode=RCIM_Cubic, leaveTangent=-2.7546, tangentMode=RCTM_User, time=1, value=0}
        ]}}})
    Scale RGB=&MakeVectorFromLinearColorRGB(Color=&ColorFromCurve(ColorCurve=DataInterface() {propertyValues {
        AlphaCurve {keys [<at 0 1 RCIM_Cubic>]}
        BlueCurve {keys [<at 0.0872 0.5>]}
        GreenCurve {keys [<at 0.0872 0.5>]}
        RedCurve {keys [<at 0.0872 0.5>]}
       }}))
    ScaleA=true
    ScaleRGB=true
   ) {module=ScaleColor002}
   &CameraOffset(Camera Offset Amount=&FloatFromCurve(
     FloatCurve=DataInterface() {propertyValues {Curve {keys [{interpMode=RCIM_Cubic, tangentMode=RCTM_User, time=1, value=5}]}}}
     CurveIndex=$Emitter.Scalability.DistanceFraction
     Scale Curve=&MaskFloatBySpawnGroup(Masked Float=50, Passthrough Float=181.7789)
    ))
  ]
  rendererClasses [&NiagaraSpriteRendererProperties]
  renderers [{rendererClass=&NiagaraSpriteRendererProperties}]
 }
 Smoke_SphereCenter001 {
  bEnabled=false
  emitterUpdateScript [
   &EmitterState()
   &SpawnBurst_Instantaneous(Spawn Count=10, Spawn Time=0.5, Use Spawn Probability=false, UseLoopCountLimit=false)
  ]
  particleSpawnScript [
   &InitializeParticle(
    Color=&RandomRangeLinearColor(Minimum=<rgba 0.25 0.1681 0.1296 1>, Maximum=<rgba 0.3 0.2017 0.1555 1>)
    Lifetime=&RandomRangeFloat(Minimum=1.5, Maximum=2)
    Sprite Rotation Angle=&RandomRangeFloat(Minimum=-25, Maximum=25)
    Sprite Rotation Mode=::NewEnumerator0
    Sprite Size Mode=::NewEnumerator1
    Uniform Sprite Size Max=1600
    Uniform Sprite Size Min=1500
    UsePositionOffset=false
    Write Material Random=true
   )
   &ShapeLocation(
    MaskSpawnGroup=false
    Offset Mode=::NewEnumerator0
    Override Seed=true
    Random Seed=126
    Randomness Mode:ENiagaraRandomnessMode=::NewEnumerator1
    Sphere Radius=1192.1179
    UseOwnerScale=true
   )
   &AddVelocity(
    Module.Use Falloff Distance=false
    Origin Offset=<0 0 -451.717>
    Override Seed=false
    Velocity Mode=::NewEnumerator1
    Velocity Speed=&RandomRangeFloat(Minimum=2811.1414, Maximum=4059.1914)
   )
   /SetVariables(Particles.Velocity=&Multiply_Vector(A=$Particles.Velocity, B=<0.5 0.5 1.3344>)) {isSetParameters=true, module=SetVariables_#2}
  ]
  particleUpdateScript [
   &ParticleState()
   &GravityForce(Gravity=&RandomRangeVector(Minimum=<0 0 -137.6014>, Maximum=<0 0 -176.7234>))
   &Drag(Drag=&RandomRangeFloat(Minimum=1.5, Maximum=2), Use Linear Drag=true, Use Rotational Drag=true)
   &SolveForcesAndVelocity(Clamp Velocity=false, Limit Acceleration=false)
   &ScaleSpriteSize(Uniform Curve Sprite Scale=DataInterface() {propertyValues {Curve {keys [
        {arriveTangent=3.9789, interpMode=RCIM_Cubic, leaveTangent=3.9789, tangentMode=RCTM_User, time=0, value=0.5}
        {arriveTangent=1.5, interpMode=RCIM_Cubic, leaveTangent=1.5, time=0.1458, value=1.7145}
        {arriveTangent=0.0224, interpMode=RCIM_Cubic, leaveTangent=0.0224, tangentMode=RCTM_User, time=1, value=2}
       ]}}})
   &DynamicMaterialParameters(
    Index 0 Param 1=&Lerp_Float(
     A=1
     B=2
     Alpha=&FloatFromCurve(FloatCurve=DataInterface() {propertyValues {Curve {keys [<at 0 0 RCIM_Cubic>,<at 1 1 RCIM_Cubic>]}}})
    )
    Module.Index0Param0Write=true
    Module.Index0Param1Write=true
    Module.Index0Param2Write=true
    Module.Index0Param3Write=true
   )
   &SubUVAnimation(
    Frames Per Second:NiagaraInt32=&RandomRangeInt(Minimum=20, Maximum=20)
    Playback Mode:ENiagara_FPSPlayrate=::NewEnumerator1
    Playback Scale Curve=DataInterface() {propertyValues {Curve {keys [
        {interpMode=RCIM_Cubic, tangentMode=RCTM_User, time=0, value=3}
        {arriveTangent=-1.2656, interpMode=RCIM_Cubic, leaveTangent=-1.2656, tangentMode=RCTM_User, time=1, value=0}
       ]}}}
    Random Start Frame=true
    "Sprite Renderer "=DataInterface() {propertyValues {SpriteRenderer=&self:Smoke_SphereCenter001_0.NiagaraSpriteRendererProperties_0}}
    Start Frame Offset:NiagaraInt32=&RandomRangeInt(Minimum=0, Maximum=6)
    Start Frame Range Override=&RandomRangeInt(Minimum=0, Maximum=0)
    SubUV Animation Mode:ENiagara_SubUVLookupModeV2=::NewEnumerator3
    UseEndFrame=true
    UsePlayRateScale=true
    UseStartFrame=true
   )
   &ScaleColor(
    Scale RGB=&MakeVectorFromLinearColorRGB(Color=&FNTimeOfDayColor(
      Night Color=<rgba 0.6 0.7243 1 1>
      Morning Color=<rgba 0.7 0.6725 0.6692 1>
      Day Color=<rgba 1 0.8103 0.65 1>
      Evening Color=<rgba 0.8 0.7149 0.6539 1>
     ))
    ScaleA=true
    ScaleRGB=true
   ) {module=ScaleColor001}
   &ScaleColor(
    Scale Alpha=&FloatFromCurve(FloatCurve=DataInterface() {propertyValues {Curve {keys [
         {arriveTangent=0.7573, interpMode=RCIM_Cubic, leaveTangent=0.7573, tangentMode=RCTM_User, time=0, value=0}
         {arriveTangent=-0.3388, interpMode=RCIM_Cubic, leaveTangent=-0.3388, tangentMode=RCTM_User, time=0.0931, value=1}
         {arriveTangent=-2.7546, interpMode=RCIM_Cubic, leaveTangent=-2.7546, tangentMode=RCTM_User, time=1, value=0}
        ]}}})
    Scale RGB=&MakeVectorFromLinearColorRGB(Color=&ColorFromCurve(ColorCurve=DataInterface() {propertyValues {
        AlphaCurve {keys [<at 0 1 RCIM_Cubic>]}
        BlueCurve {keys [<at 0.0872 0.5>]}
        GreenCurve {keys [<at 0.0872 0.5>]}
        RedCurve {keys [<at 0.0872 0.5>]}
       }}))
    ScaleA=true
    ScaleRGB=true
   )
   &CameraOffset(Camera Offset Amount=&FloatFromCurve(
     FloatCurve=DataInterface() {propertyValues {Curve {keys [{interpMode=RCIM_Cubic, tangentMode=RCTM_User, time=1, value=5}]}}}
     CurveIndex=$Emitter.Scalability.DistanceFraction
     Scale Curve=181.7789
    ))
  ]
  rendererClasses [&NiagaraSpriteRendererProperties]
  renderers [{rendererClass=&NiagaraSpriteRendererProperties}]
 }
 dirt002 {
  emitterUpdateScript [
   &EmitterState()
   &SpawnBurst_Instantaneous(Spawn Count=172, Spawn Time=0.5, Use Spawn Probability=true, UseLoopCountLimit=false)
  ]
  particleSpawnScript [
   &InitializeParticle(
    Color Maximum=<rgba 0.2664 0.2232 0.2016 1>
    Color Minimum=<rgba 0.107 0.0908 0.0844 1>
    Color Mode=::NewEnumerator2
    Lifetime=&Multiply_Float(
     A=&RandomRangeFloat(Minimum=1.5, Maximum=2.5)
     B=&MakeCustomFloatFromBool(Bool=$User.IsUltra, True Float=2, False Float=1.5)
    )
    Sprite Rotation Mode=::NewEnumerator2
    Sprite Size Mode=::NewEnumerator3
    Uniform Sprite Size=&Lerp_Float(
     A=1350.3627
     B=800
     Alpha=&FloatFromCurve(
      FloatCurve=DataInterface() {propertyValues {Curve {keys [
          {interpMode=RCIM_Cubic, tangentMode=RCTM_User, time=0, value=0}
          {arriveTangent=0.835, interpMode=RCIM_Cubic, leaveTangent=0.835, tangentMode=RCTM_User, time=0.8, value=0.2}
          {arriveTangent=4.896, interpMode=RCIM_Cubic, leaveTangent=4.896, tangentMode=RCTM_User, time=1, value=1}
         ]}}}
      CurveIndex=&RandomRangeFloat(Minimum=0, Maximum=1)
      Scale Curve=7.0905
     )
    )
    UsePositionOffset=false
    Write Material Random=true
   )
   &ShapeLocation(
    Cylinder Height=0
    Cylinder Radius=&Multiply_Float(A=$User.FX_Scalar, B=207.3457)
    MaskSpawnGroup=false
    Offset=<-100 0 0>
    Offset Mode=::NewEnumerator0
    Override Seed=false
    Rotation Axis=<0 0 1>
    Rotation Mode=::NewEnumerator1
    Shape Primitive=::NewEnumerator1
    Surface Only=true
    Surface Only Band Thickness=1
    Use Endcaps In Surface Only Mode=false
    UseOwnerScale=true
   )
   &AddVelocity(Velocity=&MakeVector(
     X=0
     Y=0
     Z=&Multiply_Float(A=&RandomRangeFloat(Minimum=0.34, Maximum=1), B=6772.8267)
    )) {module=AddVelocity001}
   &AddVelocity(Module.Use Falloff Distance=false, Override Seed=false, Velocity Mode=::NewEnumerator1, Velocity Speed=3008.6155)
   &CameraOffset(Camera Offset Amount=23.4162)
  ]
  particleUpdateScript [
   &ParticleState()
   &ApplyOwnerScaleToAttributes(Scale Initial Sprite Size=true)
   &GravityForce(Gravity=&MakeVector(
     X=0
     Y=0
     Z=&Multiply_Float(
      A=&Multiply_Float(A=&RandomRangeFloat(Minimum=1, Maximum=1), B=-7360.978)
      B=&MakeCustomFloatFromBool(Bool=$User.IsUltra, True Float=1, False Float=1)
     )
    ))
   &Drag(Use Linear Drag=true, Use Rotational Drag=false)
   &SolveForcesAndVelocity(Clamp Velocity=false, Limit Acceleration=false)
   &ScaleSpriteSize(
    Scale Sprite Size Mode=::NewEnumerator0
    Uniform Scale Factor=&Lerp_Float(
     A=0.7
     B=1.5
     Alpha=&FloatFromCurve(FloatCurve=DataInterface() {propertyValues {Curve {keys [
          {arriveTangent=4.1785, interpMode=RCIM_Cubic, leaveTangent=4.1785, tangentMode=RCTM_User, time=0, value=0}
          {arriveTangent=1.1179, interpMode=RCIM_Cubic, leaveTangent=1.1179, tangentMode=RCTM_User, time=0.2042, value=0.6228}
          {arriveTangent=0.0224, interpMode=RCIM_Cubic, leaveTangent=0.0224, tangentMode=RCTM_User, time=1, value=1}
         ]}}})
    )
   )
   &ScaleColor(
    Scale Alpha=&FloatFromCurve(FloatCurve=DataInterface() {propertyValues {Curve {keys [
         {arriveTangent=-0.2611, interpMode=RCIM_Cubic, leaveTangent=-0.2611, tangentMode=RCTM_User, time=0, value=1}
         {arriveTangent=-1.7065, interpMode=RCIM_Cubic, leaveTangent=-1.7065, tangentMode=RCTM_User, time=1, value=0}
        ]}}})
    Scale RGB=&VectorFromFloat(Value=&RandomRangeFloat(Minimum=1, Maximum=1))
    ScaleA=true
    ScaleRGB=true
   )
   &ScaleColor(
    Scale RGB=&MakeVectorFromLinearColorRGB(Color=&FNTimeOfDayColor(
      Night Color=<rgba 0.36 0.4346 0.6 1>
      Morning Color=<rgba 0.7 0.4909 0.3957 1>
      Day Color=<rgba 0.9 0.7568 0.6358 1>
      Evening Color=<rgba 0.5 0.4468 0.4087 1>
     ))
    ScaleA=true
    ScaleRGB=true
   ) {module=ScaleColor001}
  ]
  rendererClasses [&NiagaraSpriteRendererProperties]
  renderers [{rendererClass=&NiagaraSpriteRendererProperties}]
 }
}
summary {
 systemName=NS_TidalCrane_Event_RepositionInto
 emitters [
  {emitterName=Smoke_SphereCenter, rendererClasses [&NiagaraSpriteRendererProperties]}
  {emitterName=Sharp_Rocks, rendererClasses [&NiagaraMeshRendererProperties]}
  {bEnabled=false, emitterName=Sharp_Rocks001, rendererClasses [&NiagaraMeshRendererProperties]}
  {bEnabled=false, emitterName=Smoke_SphereCenter001, rendererClasses [&NiagaraSpriteRendererProperties]}
  {emitterName=dirt002, rendererClasses [&NiagaraSpriteRendererProperties]}
  {emitterName=Ground_Cracks, rendererClasses [&NiagaraDecalRendererProperties]}
 ]
 userVariables [
  User.bIsFrontEndPreview {type=NiagaraBool}
  User.BurstDirection {type=Vector3f, description="", defaultValue {struct=Vector3f, value=<0 0 1>}}
  User.Crack Orientations {
   type=NiagaraDataInterfaceArrayQuat
   description=""
   defaultValue {
    struct=NiagaraExt_VariableValue_DataInterface
    value {
     dataInterface=&self:NiagaraDataInterfaceArrayQuat_0
     dataInterfaceClass=/Script/Niagara.NiagaraDataInterfaceArrayQuat
    }
   }
  }
  User.Crack Positions {
   type=NiagaraDataInterfaceArrayFloat3
   description=""
   defaultValue {
    struct=NiagaraExt_VariableValue_DataInterface
    value {
     dataInterface=&self:NiagaraDataInterfaceArrayFloat3_0
     dataInterfaceClass=&NiagaraDataInterfaceArrayFloat3
    }
   }
  }
  User.ExplodeOnWater {type=NiagaraBool}
  User.FirstBurst {type=NiagaraBool, defaultValue=true}
  User.FX_Scalar {type=NiagaraFloat, defaultValue=1, description=""}
  User.IsUltra {type=NiagaraBool, defaultValue=true}
  User.LifeTimeDuration {type=NiagaraFloat, defaultValue=5, description=""}
  User.Water {type=NiagaraBool}
  User.WaterFX {type=NiagaraBool}
 ]
}
